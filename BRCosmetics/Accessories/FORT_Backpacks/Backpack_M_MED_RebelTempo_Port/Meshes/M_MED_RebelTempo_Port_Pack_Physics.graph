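# Physics asset for the RebelTempo Port backpack cosmetic's skeletal mesh, supplying simple collision and simulation bodies for the pack and its strap.
# Two capsule bodies cover the pack bone and the longer dyn_strap bone.
# A single constraint joins the strap to the pack with very tight swing and twist limits, so any simulated strap motion stays small.
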PhysicsAsset /BRCosmetics/Accessories/FORT_Backpacks/Backpack_M_MED_RebelTempo_Port/Meshes/M_MED_RebelTempo_Port_Pack_Physics
bodies {
 dyn_strap [{shapeType=Capsule, length=25.8037, radius=2.3403, center=<13.4845 0.004 0.0014>, rotation=<rot 0 89.9452 -89.9443>}]
 pack [{shapeType=Capsule, length=12.6622, radius=2.2002, shapeName=pack_capsule, center=<-0.0349 1.3539 0.0404>, rotation=<rot 0 -90.8954 -89.7927>}]
}
constraints [{bone1Name=dyn_strap, bone2Name=pack, swing1LimitDegrees=5, swing2LimitDegrees=3, twistLimitDegrees=1}]
