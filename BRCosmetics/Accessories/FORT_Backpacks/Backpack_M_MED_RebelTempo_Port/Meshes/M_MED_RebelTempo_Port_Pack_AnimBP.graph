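# Animation Blueprint for the RebelTempo backpack cosmetic, adding strap offsets and secondary motion.
# One branch rotates the pack while falling, offsets dyn_strap, then applies half-alpha RigidBody simulation; another runs RigidBody alone.
# A bool blend with no driving variable joins them; base_strap tweaks are disconnected.
# The EventGraph maps owner spine yaw to pack and axe offsets, which the AnimGraph never reads.
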AnimBlueprint /BRCosmetics/Accessories/FORT_Backpacks/Backpack_M_MED_RebelTempo_Port/Meshes/M_MED_RebelTempo_Port_Pack_AnimBP
target_skeleton=&/M_MED_RebelTempo_Port_Pack_Skeleton
graphs [
 AnimGraph {nodes [
   n1 = "Transform (Modify) Bone - Bone: base_strap"!(ComponentPose, bAlphaBoolEnabled:bool=True, Translation="0,-3,0", Rotation="0,8,6", Scale="1,1,1") -> (Pose) {
    alphaInputType=Bool
    rotationMode=BMM_Additive
    translationMode=BMM_Additive
    boneToModify {boneName=base_strap}
    rotation {roll=6, yaw=8}
    translation {y=-3}
   }
   n2 = "Transform (Modify) Bone - Bone: base_strap"(ComponentPose=@n1, bAlphaBoolEnabled=True, Translation="0,-3,0", Rotation="0,8,6", Scale="1,1,1") {
    alphaInputType=Bool
    rotationMode=BMM_Additive
    translationMode=BMM_Additive
    boneToModify {boneName=base_strap}
    rotation {roll=6, yaw=8}
    translation {y=-3}
   }
   "Transform (Modify) Bone - Bone: base_strap"(ComponentPose=@n2, bAlphaBoolEnabled=false, Translation="0,0,0", Rotation="10,8,6", Scale="1,1,1") {
    alphaInputType=Bool
    rotationMode=BMM_Additive
    translationMode=BMM_Additive
    translationSpace=BCS_BoneSpace
    boneToModify {boneName=base_strap}
    rotation=<rot 10 8 6>
   }
   "Transform (Modify) Bone - Bone: dyn_strap"!(ComponentPose, bAlphaBoolEnabled:bool=false, Translation="0,0,0", Rotation="20,0,0", Scale="1,1,1") -> (Pose) {
    alphaInputType=Bool
    rotationMode=BMM_Additive
    translationMode=BMM_Additive
    translationSpace=BCS_BoneSpace
    boneToModify {boneName=dyn_strap}
    rotation {pitch=20}
   }
   n3 = "Transform (Modify) Bone - Bone: pack"!(
    ComponentPose=Local To Component(LocalPose=Use cached pose 'input pose'())
    bAlphaBoolEnabled:bool=Get bIsFalling()
    Translation="0,0,0"
    Rotation="0,8,3"
    Scale
   ) -> (Pose) {
    alphaInputType=Bool
    rotationMode=BMM_Additive
    translationMode=BMM_Additive
    translationSpace=BCS_BoneSpace
    boneToModify {boneName=pack}
    rotation {roll=3, yaw=8}
   }
   n4 = "Transform (Modify) Bone - Bone: dyn_strap"(ComponentPose=@n3, bAlphaBoolEnabled=True, Translation="0,0,0", Rotation="20,0,0", Scale="1,1,1") {
    alphaInputType=Bool
    rotationMode=BMM_Additive
    translationMode=BMM_Additive
    translationSpace=BCS_BoneSpace
    boneToModify {boneName=dyn_strap}
    rotation {pitch=20}
   }
   n5 = RigidBody!(ComponentPose=@n4, Alpha=0.5, OverrideWorldGravity:Vector="0,0,-980", ExternalForce) -> (Pose) {
    bClampLinearTranslationLimitToRefPose=true
    bOverrideWorldGravity=true
    overridePhysicsAsset=&/M_MED_RebelTempo_Port_Pack_Physics
    simulationSpace=BaseBoneSpace
    baseBoneRef {boneName=root}
    componentAppliedLinearAccClamp=<0 0 0>
    simSpaceSettings {
     maxAngularAcceleration=10
     maxAngularVelocity=1
     maxLinearAcceleration=700
     maxLinearVelocity=500
     worldAlpha=1
     externalLinearDragV=<0.2 0.2 0.2>
    }
   }
   n6 = RigidBody(ComponentPose=Local To Component(LocalPose=Use cached pose 'input pose'()), OverrideWorldGravity="0,0,-980") {
    bClampLinearTranslationLimitToRefPose=true
    bOverrideWorldGravity=true
    overridePhysicsAsset=&M_MED_RebelTempo_Port_Pack_Physics
    simulationSpace=BaseBoneSpace
    baseBoneRef {boneName=root}
    componentAppliedLinearAccClamp=<0 0 0>
    simSpaceSettings {
     maxAngularAcceleration=1000
     maxAngularVelocity=1
     maxLinearAcceleration=700
     maxLinearVelocity=500
     worldAlpha=1
     externalLinearDragV=<0.2 0.2 0.2>
    }
   }
   Output Pose(Result=Blend Poses by bool(BlendPose_0=Component To Local(ComponentPose=@n6), BlendPose_1=Component To Local(ComponentPose=@n5))) {layerGroup=None}
   Save cached pose 'input pose'(Pose=Input Pose()) {cacheName=input pose}
   Get bIsSkyDiving()
   Get bIsSkydiveDiveMode()
   Get bIsSkydiveFloating()
  ]}
 EventGraph {nodes [
   Event BlueprintUpdateAnimation()
   > Sequence()
   n1 = GetOwnerSkeletalMeshComponent()
   n2 = BreakRotator(InRot=BreakTransform(InTransform=GetDeltaTransformFromRefPose(self=@n1, BoneName=spine_05, BaseName=spine_01)).Rotation)
   > Set Backpack Rotate(Backpack Rotate:Rotator=MakeRotator(
     Roll=MapRangeClamped(Value=@n2.Yaw, InRangeA=0, InRangeB=-15, OutRangeB=10)
     Yaw=MapRangeClamped(Value=@n2.Yaw, InRangeA=0, InRangeB=-30, OutRangeB=9)
    )) -> (Output_Get:Rotator)
   > Set Backpack Translation(Backpack Translation:Vector=MakeVector(Y=MapRangeClamped(Value=@n2.Yaw, InRangeA=0, InRangeB=-25, OutRangeB=-9))) -> (Output_Get:Vector)
   > Set Axe Rotate(Axe Rotate:Rotator=MakeRotator(Pitch=MapRangeClamped(Value=@n2.Yaw, InRangeA=0, InRangeB=-30, OutRangeB=-6))) -> (Output_Get:Rotator)
   TryGetPawnOwner()
  ]}
]
variables [
 Axe Rotate {type=Rotator}
 Backpack Rotate {type=Rotator}
 Backpack Translation {type=Vector}
]
